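# Niagara skin effect for the EssayViewPier male player body: orange sparks and glow sprites attached to skeletal-mesh bones, mainly the arms.
# ArmSparksMovement spawns per distance travelled, while looping bursts drive two staggered GPU spark emitters and a pulsing ParticleDecal001 sprite.
# Follow emitters spawn secondary sparks from burst particles, using the partition index stored in Particles.Scale.X to resample the bone each update.
# Calf fire and spark emitters are disabled.
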NiagaraSystem /BRCosmetics/Characters/Player/Male/Medium/Bodies/M_MED_EssayViewPier/FX/NS_EssayViewPier_Skin
dependencies {
 usedDataInterfaces [
  &NiagaraDataInterfaceArrayFloat3
  &NiagaraDataInterfaceCurve
  &NiagaraDataInterfaceArrayInt32
  &NiagaraDataInterfaceArrayFloat
  &NiagaraDataInterfaceParticleRead
  &NiagaraDataInterfaceColorCurve
  &NiagaraDataInterfaceSkeletalMesh
 ]
 usedDynamicInputs []
 usedModules [
  &self:SystemSpawnScript.SystemScriptSource.SystemScriptGraph.NiagaraNodeAssignment_0.SetVariables_#1
  &SystemState
 ]
}
emitters {
 ArmSparksMovement {
  emitterUpdateScript [
   &EmitterState(Life Cycle Mode=::NewEnumerator1, Loop Behavior=::NewEnumerator1, Loop Duration=1.5, UseLoopDelay=false)
   &SpawnPerUnit(
    Max Movement Threshold=1000
    Module.Use Max Movement Threshold=true
    Module.Use Movement Tolerance=false
    Spawn Spacing=10
    Use Spawn Probability=false
   )
  ]
  particleSpawnScript [
   &PartitionParticles()
   &InitializeParticle(
    Color=<rgba 3 0.7919 0.3503 1>
    Lifetime=0.3
    Sprite Size Max=<0.5 5>
    Sprite Size Min=<0.5 3>
    Sprite Size Mode=::NewEnumerator2
    UsePositionOffset=false
    Write Material Random=true
   )
   &SkeletalMeshLocation(
    Bone / Socket Index=&SelectIntFromArray(Direct Array Index=$Particles.PartitionParticles.Partition, Int Selection Array=DataInterface() {propertyValues {IntData [2,3]}})
    Bone Sampling Mode=::NewEnumerator4
    MaskSpawnGroup=false
    Sampled Position Offset=&AddVectorToPosition(
     Position=&SelectVectorFromArray(
      Direct Array Index=$Particles.PartitionParticles.Partition
      Vector Selection Array=DataInterface() {propertyValues {FloatData [<15 2 0>, <-15 -2 0>]}}
     )
     Vector=&RandomRangeVector(Minimum=<10 0 0>, Maximum=<-10 1 1>)
    )
    Skeletal Mesh=$System.Skeletal Mesh
   )
   &ShapeLocation(MaskSpawnGroup=false, Override Seed=false, Rotation Axis=<0 0 1>, Rotation Mode=::NewEnumerator1, Sphere Radius=3, UseOwnerScale=true)
   &CameraOffset(Camera Offset Amount=5)
   &AddVelocity(Velocity=<0 -10 0>)
   /SetVariables(Particles.RandomLerp:NiagaraFloat=&MakeCustomFloatFromBool(Bool=&RandomBool(), True Float=0, False Float=1)) {isSetParameters=true}
  ]
  particleUpdateScript [
   &ParticleState()
   /SetVariables(Particles.Glow03_Color=&MakeLinearColorFromVectorAndFloat(
     "Vector (RGB)"=&MakeVectorFromLinearColorRGB(Color=&MultiplyLinearColor(
       Linear Color=&LerpLinearColors(StartColor=<rgba 1 0.0833 0 0.2>, EndColor=<rgba 1 0.0833 0 0.2>, LerpFactor=$Particles.RandomLerp)
       Scale Factor=&MakeVector4_FromFloat(Float=1)
      ))
     "Float (Alpha)"=&Multiply_Float(A=0.3, B=1)
    )) {isSetParameters=true, module=SetVariables_#2}
   /SetVariables(Particles.Glow03_Size=&Multiply_Vector2D_ByFloat(
     Float=&FloatFromCurve(
      FloatCurve=DataInterface() {propertyValues {Curve {keys [
          <at 0 0 RCIM_Cubic>
          <at 0.2 1.3 RCIM_Cubic>
          {arriveTangent=-0.1423, interpMode=RCIM_Cubic, leaveTangent=-0.1423, tangentMode=RCTM_User, time=1, value=0}
         ]}}}
      Scale Curve=12
     )
     Vector2D=<0.3 1.2>
    )) {isSetParameters=true, module=SetVariables_#3}
   &ScaleColor(
    Scale Alpha=&FloatFromCurve(FloatCurve=DataInterface() {propertyValues {Curve {keys [<at 0.3 1>,<at 0.4 1>,<at 1 1>]}}})
    ScaleA=true
    ScaleRGB=true
   )
   &ScaleSpriteSize(Scale Factor=&Vector2DFromFloat(Value=&FloatFromCurve(
      FloatCurve=DataInterface() {propertyValues {Curve {keys [
          <at 0 0 RCIM_Cubic>
          {interpMode=RCIM_Cubic, leaveTangent=-3.0296, tangentMode=RCTM_Break, time=0.2, value=1.2}
          {arriveTangent=-2.8894, interpMode=RCIM_Cubic, leaveTangent=-2.8894, tangentMode=RCTM_User, time=1, value=0.1}
         ]}}}
      Scale Curve=0.4
     ))) {module=ScaleSpriteSize001}
   &Drag(Drag=4, Use Linear Drag=true, Use Rotational Drag=false)
   &InheritSourceMovement(
    Applied Position Scale=&Multiply_VectorByFloat(Vector=<1 1 1>, Float=0.7)
    ApplyForce=false
    ApplyPosition=true
    ApplyPositionFalloff=false
   )
   &SolveForcesAndVelocity(Clamp Velocity=false, Limit Acceleration=false)
  ]
  rendererClasses [&NiagaraSpriteRendererProperties]
  renderers [
   {rendererClass=&NiagaraSpriteRendererProperties}
   {rendererClass=&NiagaraSpriteRendererProperties, rendererIndex=1}
  ]
 }
 BurstArmSparks {
  simTarget=GPUComputeSim
  emitterUpdateScript [
   &InterpolateOverTime(
    Rate Of Change=5
    Target Float Value:NiagaraFloat=&MakeCustomFloatFromBool(
     Bool=&SetBoolByFloatComparison(A=&VectorLength(Vector=$Engine.Owner.Velocity), B=100)
     True Float=0
     False Float=1
    )
   )
   &EmitterState(Life Cycle Mode=::NewEnumerator1, Loop Behavior=::NewEnumerator1, Loop Duration=1, UseLoopDelay=false)
   &SpawnBurst_Instantaneous(Spawn Count=100, Use Spawn Probability=false, UseLoopCountLimit=false)
   &SpawnRate(Spawn Probability=$StackContext.InterpolateOverTime.InterpolateOverTime_Float.Moving Average, SpawnRate=30, Use Spawn Probability=true) {enabled=false}
  ]
  particleSpawnScript [
   &PartitionParticles(Number Of Partitions=4)
   &InitializeParticle(
    Color=<rgba 3 0.7919 0.3503 1>
    Lifetime Max=1.5
    Lifetime Min=0.6
    Lifetime Mode=::NewEnumerator1
    Sprite Size Max=<0.6 3>
    Sprite Size Min=<0.4 5>
    Sprite Size Mode=::NewEnumerator2
    UsePositionOffset=false
    Write Material Random=true
   )
   &ShapeLocation(
    MaskSpawnGroup=false
    Override Seed=false
    Rotation Axis=<0 0 1>
    Rotation Mode=::NewEnumerator1
    Shape Origin=&VectorFromFloat(Value=0)
    Sphere Radius=2
    UseOwnerScale=true
   )
   &CameraOffset(Camera Offset Amount=-10)
   &AddVelocity()
   /SetVariables(Particles.SizeScale:Vector2f=&MakeVector2D(
     X=1
     Y=&FloatFromCurve(FloatCurve=DataInterface() {propertyValues {Curve {keys [{arriveTangent=-0.3799, interpMode=RCIM_Cubic, leaveTangent=-0.3799, tangentMode=RCTM_User, time=0, value=1}, <at 1 0>]}}})
    )) {isSetParameters=true, module=SetVariables_#4}
   /SetVariables(Particles.RandomLerp:NiagaraFloat=&MakeCustomFloatFromBool(Bool=&RandomBool(), True Float=0, False Float=1)) {isSetParameters=true, module=SetVariables_#5}
   /SetVariables(Particles.Scale=&MakeVector(X=&MakeFloatFromInt(Int=$Particles.PartitionParticles.Partition), Y=0, Z=0)) {isSetParameters=true, module=SetVariables_#6}
  ]
  particleUpdateScript [
   &SkeletalMeshLocation(
    Bone / Socket Index=&SelectIntFromArray(Direct Array Index=$Particles.PartitionParticles.Partition, Int Selection Array=DataInterface() {propertyValues {IntData [2,3,2,3]}})
    Bone Sampling Mode=::NewEnumerator4
    MaskSpawnGroup=false
    Position Sampling=::NewEnumerator2
    Sampled Position Offset=&AddVectorToPosition(
     Position=&AddVectorToPosition(
      Position=&SelectVectorFromArray(
       Direct Array Index=$Particles.PartitionParticles.Partition
       Vector Selection Array=DataInterface() {propertyValues {FloatData [<20 0 0>, <-20 0 0>, <20 0 0>, <-20 0 0>]}}
      )
      Vector=&AddVectorToPosition(Position=&RandomRangeVector(Minimum=<0 0 0>, Maximum=<0 0 0>), Vector=<0 0 0>)
     )
     Vector=<0 0 0>
    )
    Skeletal Mesh=$System.Skeletal Mesh
   )
   &ParticleState()
   /SetVariables(Particles.Glow03_Color=&MakeLinearColorFromVectorAndFloat(
     "Vector (RGB)"=&MakeVectorFromLinearColorRGB(Color=&MultiplyLinearColor(
       Linear Color=&LerpLinearColors(StartColor=<rgba 1 0.0833 0 0.2>, EndColor=<rgba 1 0.0833 0 0.2>, LerpFactor=$Particles.RandomLerp)
       Scale Factor=&MakeVector4_FromFloat(Float=1)
      ))
     "Float (Alpha)"=&Multiply_Float(
      A=0.3
      B=&MakeCustomFloatFromBool(
       Bool=&SetBoolByFloatComparison(A=&VectorLength(Vector=$Engine.Owner.Velocity), B=100)
       True Float=1
       False Float=1
      )
     )
    )) {isSetParameters=true, module=SetVariables_#2}
   /SetVariables(Particles.Glow03_Size=&Multiply_Vector2D_ByFloat(
     Float=&FloatFromCurve(
      FloatCurve=DataInterface() {propertyValues {Curve {keys [
          <at 0 0 RCIM_Cubic>
          <at 0.2 1.3 RCIM_Cubic>
          {arriveTangent=-0.1423, interpMode=RCIM_Cubic, leaveTangent=-0.1423, tangentMode=RCTM_User, time=1, value=0}
         ]}}}
      Scale Curve=14
     )
     Vector2D=&Multiply_Vector2D(A=<0.2 0.7>, B=$Particles.SizeScale)
    )) {isSetParameters=true, module=SetVariables_#3}
   &ScaleColor(
    Scale Alpha=&FloatFromCurve(
     FloatCurve=DataInterface() {propertyValues {Curve {keys [<at 0.3 1>,<at 0.4 1>,<at 1 1>]}}}
     Scale Curve=&MakeCustomFloatFromBool(
      Bool=&SetBoolByFloatComparison(A=&VectorLength(Vector=$Engine.Owner.Velocity), B=100)
      True Float=1
      False Float=1
     )
    )
    ScaleA=true
    ScaleRGB=true
   )
   &ScaleSpriteSize(Scale Factor=&Multiply_Vector2D(
     A=&Vector2DFromFloat(Value=&FloatFromCurve(
       FloatCurve=DataInterface() {propertyValues {Curve {keys [
           <at 0 0 RCIM_Cubic>
           {interpMode=RCIM_Cubic, leaveTangent=-3.0296, tangentMode=RCTM_Break, time=0.2, value=1.2}
           {arriveTangent=-2.8894, interpMode=RCIM_Cubic, leaveTangent=-2.8894, tangentMode=RCTM_User, time=1, value=0.1}
          ]}}}
       Scale Curve=0.4
      ))
     B=$Particles.SizeScale
    )) {module=ScaleSpriteSize001}
   &CurlNoiseForce(
    Noise Frequency=300
    Noise Strength=&Multiply_Float(A=200, B=1)
    Pan Noise=true
    Pan Noise Field=<0 0 1>
    Random Seed=10
   )
   &Drag(
    Drag=&FloatFromCurve(FloatCurve=DataInterface() {propertyValues {Curve {keys [<at 0 1>,<at 1 3>]}}})
    Use Linear Drag=true
    Use Rotational Drag=false
   )
   &GravityForce(Gravity=<0 0 50>)
   &GravityForce(Gravity=<0 0 -50>) {module=GravityForce001}
   &SolveForcesAndVelocity(Clamp Velocity=false, Limit Acceleration=false)
   /SetVariables(Particles.Position02:NiagaraPosition=&AddVectorToPosition(Position=$Particles.Position, Vector=Linked() {linkedVariable {name=Particles.SkeletalMeshLocation.SampledPosition, type=NiagaraPosition}})) {isSetParameters=true, module=SetVariables_#7}
  ]
  rendererClasses [&NiagaraSpriteRendererProperties]
  renderers [
   {rendererClass=&NiagaraSpriteRendererProperties}
   {rendererClass=&NiagaraSpriteRendererProperties, rendererIndex=1}
  ]
 }
 BurstArmSparks001 {
  simTarget=GPUComputeSim
  emitterUpdateScript [
   &InterpolateOverTime(
    Rate Of Change=5
    Target Float Value:NiagaraFloat=&MakeCustomFloatFromBool(
     Bool=&SetBoolByFloatComparison(A=&VectorLength(Vector=$Engine.Owner.Velocity), B=100)
     True Float=0
     False Float=1
    )
   )
   &EmitterState(Life Cycle Mode=::NewEnumerator1, Loop Behavior=::NewEnumerator1, Loop Delay=0.2, Loop Duration=1, UseLoopDelay=true)
   &SpawnBurst_Instantaneous(Spawn Count=100, Use Spawn Probability=false, UseLoopCountLimit=false)
   &SpawnRate(Spawn Probability=$StackContext.InterpolateOverTime.InterpolateOverTime_Float.Moving Average, SpawnRate=30, Use Spawn Probability=true) {enabled=false}
  ]
  particleSpawnScript [
   &PartitionParticles(Number Of Partitions=4)
   &InitializeParticle(
    Color=<rgba 3 0.7919 0.3503 1>
    Lifetime Max=1.5
    Lifetime Min=0.6
    Lifetime Mode=::NewEnumerator1
    Sprite Size Max=<0.6 3>
    Sprite Size Min=<0.4 5>
    Sprite Size Mode=::NewEnumerator2
    UsePositionOffset=false
    Write Material Random=true
   )
   &ShapeLocation(
    MaskSpawnGroup=false
    Override Seed=false
    Rotation Axis=<0 0 1>
    Rotation Mode=::NewEnumerator1
    Shape Origin=&VectorFromFloat(Value=0)
    Sphere Radius=2
    UseOwnerScale=true
   )
   &CameraOffset(Camera Offset Amount=-10)
   &AddVelocity()
   /SetVariables(Particles.SizeScale:Vector2f=&MakeVector2D(
     X=1
     Y=&FloatFromCurve(FloatCurve=DataInterface() {propertyValues {Curve {keys [{arriveTangent=-0.3799, interpMode=RCIM_Cubic, leaveTangent=-0.3799, tangentMode=RCTM_User, time=0, value=1}, <at 1 0>]}}})
    )) {isSetParameters=true, module=SetVariables_#4}
   /SetVariables(Particles.RandomLerp:NiagaraFloat=&MakeCustomFloatFromBool(Bool=&RandomBool(), True Float=0, False Float=1)) {isSetParameters=true, module=SetVariables_#5}
   /SetVariables(Particles.Scale=&MakeVector(X=&MakeFloatFromInt(Int=$Particles.PartitionParticles.Partition), Y=0, Z=0)) {isSetParameters=true, module=SetVariables_#6}
  ]
  particleUpdateScript [
   &SkeletalMeshLocation(
    Bone / Socket Index=&SelectIntFromArray(Direct Array Index=$Particles.PartitionParticles.Partition, Int Selection Array=DataInterface() {propertyValues {IntData [2,3,4,5]}})
    Bone Sampling Mode=::NewEnumerator4
    MaskSpawnGroup=false
    Position Sampling=::NewEnumerator2
    Sampled Position Offset=&AddVectorToPosition(
     Position=&AddVectorToPosition(
      Position=&SelectVectorFromArray(
       Direct Array Index=$Particles.PartitionParticles.Partition
       Vector Selection Array=DataInterface() {propertyValues {FloatData [<20 0 0>, <-20 0 0>, <20 0 0>, <-20 0 0>]}}
      )
      Vector=&AddVectorToPosition(Position=&RandomRangeVector(Minimum=<0 0 0>, Maximum=<0 0 0>), Vector=<0 0 0>)
     )
     Vector=<0 0 0>
    )
    Skeletal Mesh=$System.Skeletal Mesh
   )
   &ParticleState()
   /SetVariables(Particles.Glow03_Color=&MakeLinearColorFromVectorAndFloat(
     "Vector (RGB)"=&MakeVectorFromLinearColorRGB(Color=&MultiplyLinearColor(
       Linear Color=&LerpLinearColors(StartColor=<rgba 1 0.0833 0 0.2>, EndColor=<rgba 1 0.0833 0 0.2>, LerpFactor=$Particles.RandomLerp)
       Scale Factor=&MakeVector4_FromFloat(Float=1)
      ))
     "Float (Alpha)"=&Multiply_Float(
      A=0.3
      B=&MakeCustomFloatFromBool(
       Bool=&SetBoolByFloatComparison(A=&VectorLength(Vector=$Engine.Owner.Velocity), B=100)
       True Float=1
       False Float=1
      )
     )
    )) {isSetParameters=true, module=SetVariables_#2}
   /SetVariables(Particles.Glow03_Size=&Multiply_Vector2D_ByFloat(
     Float=&FloatFromCurve(
      FloatCurve=DataInterface() {propertyValues {Curve {keys [
          <at 0 0 RCIM_Cubic>
          <at 0.2 1.3 RCIM_Cubic>
          {arriveTangent=-0.1423, interpMode=RCIM_Cubic, leaveTangent=-0.1423, tangentMode=RCTM_User, time=1, value=0}
         ]}}}
      Scale Curve=14
     )
     Vector2D=&Multiply_Vector2D(A=<0.2 0.7>, B=$Particles.SizeScale)
    )) {isSetParameters=true, module=SetVariables_#3}
   &ScaleColor(
    Scale Alpha=&FloatFromCurve(
     FloatCurve=DataInterface() {propertyValues {Curve {keys [<at 0.3 1>,<at 0.4 1>,<at 1 1>]}}}
     Scale Curve=&MakeCustomFloatFromBool(
      Bool=&SetBoolByFloatComparison(A=&VectorLength(Vector=$Engine.Owner.Velocity), B=100)
      True Float=1
      False Float=1
     )
    )
    ScaleA=true
    ScaleRGB=true
   )
   &ScaleSpriteSize(Scale Factor=&Multiply_Vector2D(
     A=&Vector2DFromFloat(Value=&FloatFromCurve(
       FloatCurve=DataInterface() {propertyValues {Curve {keys [
           <at 0 0 RCIM_Cubic>
           {interpMode=RCIM_Cubic, leaveTangent=-3.0296, tangentMode=RCTM_Break, time=0.2, value=1.2}
           {arriveTangent=-2.8894, interpMode=RCIM_Cubic, leaveTangent=-2.8894, tangentMode=RCTM_User, time=1, value=0.1}
          ]}}}
       Scale Curve=0.4
      ))
     B=$Particles.SizeScale
    )) {module=ScaleSpriteSize001}
   &CurlNoiseForce(
    Noise Frequency=300
    Noise Strength=&Multiply_Float(A=200, B=1)
    Pan Noise=true
    Pan Noise Field=<0 0 1>
    Random Seed=10
   )
   &Drag(
    Drag=&FloatFromCurve(FloatCurve=DataInterface() {propertyValues {Curve {keys [<at 0 1>,<at 1 3>]}}})
    Use Linear Drag=true
    Use Rotational Drag=false
   )
   &GravityForce(Gravity=<0 0 50>)
   &GravityForce(Gravity=<0 0 -50>) {module=GravityForce001}
   &SolveForcesAndVelocity(Clamp Velocity=false, Limit Acceleration=false)
   /SetVariables(Particles.Position02:NiagaraPosition=&AddVectorToPosition(Position=$Particles.Position, Vector=Linked() {linkedVariable {name=Particles.SkeletalMeshLocation.SampledPosition, type=NiagaraPosition}})) {isSetParameters=true, module=SetVariables_#7}
  ]
  rendererClasses [&NiagaraSpriteRendererProperties]
  renderers [
   {rendererClass=&NiagaraSpriteRendererProperties}
   {rendererClass=&NiagaraSpriteRendererProperties, rendererIndex=1}
  ]
 }
 BurstSparksFollow {
  simTarget=GPUComputeSim
  emitterUpdateScript [
   &InterpolateOverTime(
    Rate Of Change=5
    Target Float Value:NiagaraFloat=&MakeCustomFloatFromBool(
     Bool=&SetBoolByFloatComparison(A=&VectorLength(Vector=$Engine.Owner.Velocity), B=100)
     True Float=0
     False Float=1
    )
   )
   &EmitterState(Life Cycle Mode=::NewEnumerator1, Loop Behavior=::NewEnumerator1, Loop Duration=0.5, UseLoopDelay=false)
   &SpawnParticlesFromOtherEmitter(
    Attribute Reader=DataInterface() {propertyValues {EmitterBinding {bindingMode=Other, emitterName=BurstArmSparks}}}
    Spawn Rate=&Multiply_Float(
     A=40
     B=&MakeCustomFloatFromBool(
      Bool=&SetBoolByFloatComparison(A=&VectorLength(Vector=$Engine.Owner.Velocity), B=400)
      True Float=1
      False Float=1
     )
    )
    Use Maximum Age=false
    Use Minimum Age=false
   )
  ]
  particleSpawnScript [
   &InitializeParticle(
    Color=&MakeCustomLinearColorFromBool(Bool=&RandomBool(), True Color=<rgba 100 3.3333 0 1>, False Color=<rgba 100 0.8384 0 1>)
    Lifetime Max=0.5
    Lifetime Min=0.31
    Lifetime Mode=::NewEnumerator1
    Sprite Size Max=<1 1>
    Sprite Size Min=<0.5 0.5>
    Sprite Size Mode=::NewEnumerator2
    UsePositionOffset=false
    Write Material Random=true
   )
   &SampleParticlesFromOtherEmitter(Applied Velocity Scale=0.1, Mesh Scale Sampling:ENiagara_DirectReadSamplingMode=::NewEnumerator0, Velocity Sampling=::NewEnumerator0)
   &AddVelocity(
    Cone Angle=50
    Cone Axis=<0 0 1>
    Module.UseConeFalloff=false
    Module.UseVelDistribution=false
    Override Seed=false
    Velocity Mode=::NewEnumerator2
    Velocity Speed=5
   )
   &CameraOffset()
  ]
  particleUpdateScript [
   &ParticleState()
   &SkeletalMeshLocation(
    Bone / Socket Index=&SelectIntFromArray(
     Direct Array Index=&MakeIntFromFloat(Float=&MakeFloatFromVector(Vector=$Particles.Scale, Channel=::NewEnumerator0))
     Int Selection Array=DataInterface() {propertyValues {IntData [2,3,4,5]}}
    )
    Bone Sampling Mode=::NewEnumerator4
    MaskSpawnGroup=false
    Position Sampling=::NewEnumerator2
    Sampled Position Offset=&AddVectorToPosition(
     Position=&AddVectorToPosition(
      Position=&SelectVectorFromArray(
       Direct Array Index=&MakeIntFromFloat(Float=&MakeFloatFromVector(Vector=$Particles.Scale, Channel=::NewEnumerator0))
       Vector Selection Array=DataInterface() {propertyValues {FloatData [<20 0 0>, <-20 0 0>, <20 -2 0>, <-20 2 0>]}}
      )
      Vector=&AddVectorToPosition(Position=&RandomRangeVector(Minimum=<0 0 0>, Maximum=<0 0 0>), Vector=<0 0 0>)
     )
     Vector=<0 0 0>
    )
    Skeletal Mesh=$System.Skeletal Mesh
   )
   &ScaleSpriteSize(
    Uniform Curve Scale=0.5
    Uniform Curve Sprite Scale=DataInterface() {propertyValues {Curve {
       defaultValue=1
       keys [
        {interpMode=RCIM_Cubic, tangentMode=RCTM_User, time=0, value=0}
        {arriveTangent=2, interpMode=RCIM_Cubic, leaveTangent=2, time=0.2077, value=1}
        {interpMode=RCIM_Cubic, tangentMode=RCTM_User, time=0.5, value=1}
        {tangentMode=RCTM_User, time=1, value=0}
       ]
      }}}
   )
   &Drag(Use Linear Drag=true, Use Rotational Drag=true)
   &SolveForcesAndVelocity(Clamp Velocity=false, Limit Acceleration=false)
   &ScaleColor(
    Scale Alpha=&FloatFromCurve(FloatCurve=DataInterface() {propertyValues {Curve {
        defaultValue=1
        keys [
         {interpMode=RCIM_Cubic, tangentMode=RCTM_User, time=0, value=0}
         {arriveTangent=2, interpMode=RCIM_Cubic, leaveTangent=2, time=0.1, value=1}
         {interpMode=RCIM_Cubic, tangentMode=RCTM_User, time=0.5, value=1}
         {tangentMode=RCTM_User, time=1, value=0}
        ]
       }}})
    Scale RGB=&Multiply_VectorByFloat(
     Vector=&MakeVectorFromLinearColorRGB(Color=&ColorFromCurve(ColorCurve=DataInterface() {propertyValues {
         AlphaCurve {keys []}
         BlueCurve {keys [<at 0.3911 10 RCIM_Cubic>, {interpMode=RCIM_Cubic, leaveTangent=15.4926, time=1, value=10}]}
         GreenCurve {keys [<at 0.3911 10 RCIM_Cubic>, {interpMode=RCIM_Cubic, leaveTangent=-42.8993, time=1, value=10}]}
         RedCurve {keys [<at 0.3911 10 RCIM_Cubic>, {interpMode=RCIM_Cubic, leaveTangent=-72.6613, time=1, value=10}]}
        }}))
     Float=&FloatFromCurve(FloatCurve=DataInterface() {propertyValues {Curve {keys [<at 0 1>,<at 1 1>]}}})
    )
    ScaleA=true
    ScaleRGB=true
   )
   &DynamicMaterialParameters(
    Index 0 Param 1=1
    Module.Index0Float0Write=true
    Module.Index0Float1Write=true
    Module.Index0Float2Write=true
    Module.Index0Float3Write=true
   )
   /SetVariables(Particles.Position02:NiagaraPosition=&AddVectorToPosition(Position=$Particles.Position, Vector=Linked() {linkedVariable {name=Particles.SkeletalMeshLocation.SampledPosition, type=NiagaraPosition}})) {isSetParameters=true, module=SetVariables_#8}
  ]
  rendererClasses [&NiagaraSpriteRendererProperties]
  renderers [{rendererClass=&NiagaraSpriteRendererProperties}]
 }
 BurstSparksFollow001 {
  simTarget=GPUComputeSim
  emitterUpdateScript [
   &InterpolateOverTime(
    Rate Of Change=5
    Target Float Value:NiagaraFloat=&MakeCustomFloatFromBool(
     Bool=&SetBoolByFloatComparison(A=&VectorLength(Vector=$Engine.Owner.Velocity), B=100)
     True Float=0
     False Float=1
    )
   )
   &EmitterState(Life Cycle Mode=::NewEnumerator1, Loop Behavior=::NewEnumerator1, Loop Delay=0.1, Loop Duration=0.7, UseLoopDelay=true)
   &SpawnParticlesFromOtherEmitter(
    Attribute Reader=DataInterface() {propertyValues {EmitterBinding {bindingMode=Other, emitterName=BurstArmSparks001}}}
    Spawn Rate=&Multiply_Float(
     A=40
     B=&MakeCustomFloatFromBool(
      Bool=&SetBoolByFloatComparison(A=&VectorLength(Vector=$Engine.Owner.Velocity), B=400)
      True Float=1
      False Float=1
     )
    )
    Use Maximum Age=false
    Use Minimum Age=false
   )
  ]
  particleSpawnScript [
   &InitializeParticle(
    Color=&MakeCustomLinearColorFromBool(Bool=&RandomBool(), True Color=<rgba 100 3.3333 0 1>, False Color=<rgba 100 0.8384 0 1>)
    Lifetime Max=0.5
    Lifetime Min=0.31
    Lifetime Mode=::NewEnumerator1
    Sprite Size Max=<1 1>
    Sprite Size Min=<0.5 0.5>
    Sprite Size Mode=::NewEnumerator2
    UsePositionOffset=false
    Write Material Random=true
   )
   &SampleParticlesFromOtherEmitter(Applied Velocity Scale=0.1, Mesh Scale Sampling:ENiagara_DirectReadSamplingMode=::NewEnumerator0, Velocity Sampling=::NewEnumerator0)
   &AddVelocity(
    Cone Angle=50
    Cone Axis=<0 0 1>
    Module.UseConeFalloff=false
    Module.UseVelDistribution=false
    Override Seed=false
    Velocity Mode=::NewEnumerator2
    Velocity Speed=5
   )
   &CameraOffset()
  ]
  particleUpdateScript [
   &ParticleState()
   &SkeletalMeshLocation(
    Bone / Socket Index=&SelectIntFromArray(
     Direct Array Index=&MakeIntFromFloat(Float=&MakeFloatFromVector(Vector=$Particles.Scale, Channel=::NewEnumerator0))
     Int Selection Array=DataInterface() {propertyValues {IntData [2,3,4,5]}}
    )
    Bone Sampling Mode=::NewEnumerator4
    MaskSpawnGroup=false
    Position Sampling=::NewEnumerator2
    Sampled Position Offset=&AddVectorToPosition(
     Position=&AddVectorToPosition(
      Position=&SelectVectorFromArray(
       Direct Array Index=&MakeIntFromFloat(Float=&MakeFloatFromVector(Vector=$Particles.Scale, Channel=::NewEnumerator0))
       Vector Selection Array=DataInterface() {propertyValues {FloatData [<8 0 0>, <-8 0 0>, <8 -2 0>, <-8 2 0>]}}
      )
      Vector=&AddVectorToPosition(Position=&RandomRangeVector(Minimum=<0 0 0>, Maximum=<0 0 0>), Vector=<0 0 0>)
     )
     Vector=<0 0 0>
    )
    Skeletal Mesh=$System.Skeletal Mesh
   )
   &ScaleSpriteSize(
    Uniform Curve Scale=0.5
    Uniform Curve Sprite Scale=DataInterface() {propertyValues {Curve {
       defaultValue=1
       keys [
        {interpMode=RCIM_Cubic, tangentMode=RCTM_User, time=0, value=0}
        {arriveTangent=2, interpMode=RCIM_Cubic, leaveTangent=2, time=0.2077, value=1}
        {interpMode=RCIM_Cubic, tangentMode=RCTM_User, time=0.5, value=1}
        {tangentMode=RCTM_User, time=1, value=0}
       ]
      }}}
   )
   &Drag(Use Linear Drag=true, Use Rotational Drag=true)
   &SolveForcesAndVelocity(Clamp Velocity=false, Limit Acceleration=false)
   &ScaleColor(
    Scale Alpha=&FloatFromCurve(FloatCurve=DataInterface() {propertyValues {Curve {
        defaultValue=1
        keys [
         {interpMode=RCIM_Cubic, tangentMode=RCTM_User, time=0, value=0}
         {arriveTangent=2, interpMode=RCIM_Cubic, leaveTangent=2, time=0.1, value=1}
         {interpMode=RCIM_Cubic, tangentMode=RCTM_User, time=0.5, value=1}
         {tangentMode=RCTM_User, time=1, value=0}
        ]
       }}})
    Scale RGB=&Multiply_VectorByFloat(
     Vector=&MakeVectorFromLinearColorRGB(Color=&ColorFromCurve(ColorCurve=DataInterface() {propertyValues {
         AlphaCurve {keys []}
         BlueCurve {keys [<at 0.3911 10 RCIM_Cubic>, {interpMode=RCIM_Cubic, leaveTangent=15.4926, time=1, value=10}]}
         GreenCurve {keys [<at 0.3911 10 RCIM_Cubic>, {interpMode=RCIM_Cubic, leaveTangent=-42.8993, time=1, value=10}]}
         RedCurve {keys [<at 0.3911 10 RCIM_Cubic>, {interpMode=RCIM_Cubic, leaveTangent=-72.6613, time=1, value=10}]}
        }}))
     Float=&FloatFromCurve(FloatCurve=DataInterface() {propertyValues {Curve {keys [<at 0 1>,<at 1 1>]}}})
    )
    ScaleA=true
    ScaleRGB=true
   )
   &DynamicMaterialParameters(
    Index 0 Param 1=1
    Module.Index0Float0Write=true
    Module.Index0Float1Write=true
    Module.Index0Float2Write=true
    Module.Index0Float3Write=true
   )
   /SetVariables(Particles.Position02:NiagaraPosition=&AddVectorToPosition(Position=$Particles.Position, Vector=Linked() {linkedVariable {name=Particles.SkeletalMeshLocation.SampledPosition, type=NiagaraPosition}})) {isSetParameters=true, module=SetVariables_#8}
  ]
  rendererClasses [&NiagaraSpriteRendererProperties]
  renderers [{rendererClass=&NiagaraSpriteRendererProperties}]
 }
 CalfFire {
  bEnabled=false
  emitterUpdateScript [
   &EmitterState(Life Cycle Mode=::NewEnumerator1, Loop Behavior=::NewEnumerator1, Loop Duration=2, UseLoopDelay=false)
   &SpawnBurst_Instantaneous(Spawn Count=2, Use Spawn Probability=false, UseLoopCountLimit=false)
  ]
  particleSpawnScript [
   &PartitionParticles()
   /SetVariables(Particles.VectorScale:Vector3f=&Multiply_VectorByFloat(
     Vector=&SelectVectorFromArray(
      Direct Array Index=$Particles.PartitionParticles.Partition
      Vector Selection Array=DataInterface() {propertyValues {FloatData [<1.5 1.2 1.5>, <1.5 1.2 1>]}}
     )
     Float=&SelectFloatFromArray(Direct Array Index=$Particles.PartitionParticles.Partition, Float Selection Array=DataInterface() {propertyValues {FloatData [0.4,0.5]}})
    )) {isSetParameters=true}
   &InitializeParticle(
    Color=<rgba 1 1 1 0.2>
    Lifetime=1.5
    Mesh Scale=&Multiply_Vector(
     A=&Multiply_Vector(A=<1 1 1.5>, B=$Particles.VectorScale)
     B=&VectorFromFloat(Value=$Particles.SizeScle)
    )
    Mesh Scale Mode=::NewEnumerator4
    UsePositionOffset=false
    Write Material Random=true
   ) {module=InitializeParticle001}
   &InitialMeshOrientation(Rotation=<0 0 0.25>, Use Orientation Vector=true, Use Rotation Vector=true)
   &InitialMeshOrientation(
    Rotation=&SelectVectorFromArray(
     Direct Array Index=$Particles.PartitionParticles.Partition
     Vector Selection Array=DataInterface() {propertyValues {FloatData [<0.5 0 0>, <0 0 0>]}}
    )
    Use Orientation Vector=true
    Use Rotation Vector=true
   ) {module=InitialMeshOrientation002}
   &InitialMeshOrientation(
    Rotation=&SelectVectorFromArray(
     Direct Array Index=$Particles.PartitionParticles.Partition
     Vector Selection Array=DataInterface() {propertyValues {FloatData [<-0.1 -0.1 0.05>, <0 -0.1 0>]}}
    )
    Use Orientation Vector=true
    Use Rotation Vector=true
   ) {module=InitialMeshOrientation001}
   &CameraOffset(Camera Offset Amount=-2)
  ]
  particleUpdateScript [
   /SetVariables(Particles.SizeScle:NiagaraFloat=&Multiply_Float(
     A=0.9
     B=&FloatFromCurve(FloatCurve=DataInterface() {propertyValues {Curve {keys [<at 0 1>,<at 0.5036 1>,<at 1 1>]}}})
    )) {isSetParameters=true}
   /SetVariables(Particles.Mesh_Color2:LinearColor=&MakeLinearColorFromVectorAndFloat(
     "Vector (RGB)"=&MakeVectorFromLinearColorRGB(Color=&MultiplyLinearColor(Linear Color=<rgba 1 0.6001 0.221 1>, Scale Factor=&MakeVector4_FromFloat(Float=1)))
     "Float (Alpha)"=&Multiply_Float(
      A=0.5
      B=&FloatFromCurve(
       FloatCurve=DataInterface() {propertyValues {Curve {keys [<at 0 0>,<at 0.2 1>,<at 0.4955 0.2235>,<at 1 0>]}}}
       Scale Curve=&MakeCustomFloatFromBool(Bool=$User.bisFootHidden, True Float=0, False Float=1)
      )
     )
    )) {isSetParameters=true}
   /SetVariables(Particles.Mesh_Size2:Vector3f=&Multiply_VectorByFloat(Vector=&Multiply_Vector(A=<1 0.9 1.2>, B=$Particles.VectorScale), Float=$Particles.SizeScle)) {isSetParameters=true}
   &SkeletalMeshLocation(
    Bone / Socket Index=$Particles.PartitionParticles.Partition
    Bone Sampling Mode=::NewEnumerator4
    MaskSpawnGroup=false
    Sampled Position Offset=&SelectVectorFromArray(
     Direct Array Index=$Particles.PartitionParticles.Partition
     Vector Selection Array=DataInterface() {propertyValues {FloatData [<-15 3 -2>, <15 0 0>]}}
    )
    Skeletal Mesh=$System.Skeletal Mesh
   )
   &ParticleState()
   &InterpolateOverTime(Rate Of Change=15, Target Vector Value:NiagaraPosition=$Particles.Position, Variable Type=::NewEnumerator2) {module=InterpolateOverTime001}
   /SetVariables(Particles.Interpolate Dynamic=&MakeFloat4FromVectorAndFloat("Vector (RGB)"=$Particles.InterpolateOverTime001.InterpolateOverTime_V3_.Moving Average, "Float (Alpha)"=1)) {isSetParameters=true}
   &SolveForcesAndVelocity(Clamp Velocity=false, Limit Acceleration=false)
   &ScaleColor(
    Scale Alpha=&Multiply_Float(
     A=&FloatFromCurve(
      FloatCurve=DataInterface() {propertyValues {Curve {keys [<at 0 0>,<at 0.2 1>,<at 0.8 1>,<at 1 0>]}}}
      Scale Curve=&MakeCustomFloatFromBool(Bool=$User.bisFootHidden, True Float=0, False Float=1)
     )
     B=2
    )
    Scale RGB=&Multiply_VectorByFloat(
     Vector=<1 1 1>
     Float=&MakeCustomFloatFromBool(Bool=$User.bIsFrontEndPreview, True Float=0.5, False Float=2)
    )
    ScaleA=true
    ScaleRGB=true
   )
   &DynamicMaterialParameters(
    Index 0 Param 1=&SelectFloatFromArray(Direct Array Index=$Particles.PartitionParticles.Partition, Float Selection Array=DataInterface() {propertyValues {FloatData [0,0.5]}})
    Index 0 Param 2=1
    Index 0 Param 3=&SelectFloatFromArray(Direct Array Index=$Particles.PartitionParticles.Partition, Float Selection Array=DataInterface() {propertyValues {FloatData [0.9,0.7]}})
    Index 0 Param 4=&SelectFloatFromArray(Direct Array Index=$Particles.PartitionParticles.Partition, Float Selection Array=DataInterface() {propertyValues {FloatData [1,0.8]}})
    Module.Index0Float0Write=true
    Module.Index0Float1Write=true
    Module.Index0Float2Write=true
    Module.Index0Float3Write=true
   ) {module=DynamicMaterialParameters001}
  ]
  rendererClasses [&NiagaraMeshRendererProperties]
  renderers [
   {rendererClass=&NiagaraMeshRendererProperties, rendererIndex=0}
   {rendererClass=&NiagaraMeshRendererProperties, rendererIndex=1}
  ]
 }
 CalfSparks {
  bEnabled=false
  emitterUpdateScript [
   &InterpolateOverTime(
    Rate Of Change=5
    Target Float Value:NiagaraFloat=&MakeCustomFloatFromBool(
     Bool=&SetBoolByFloatComparison(A=&VectorLength(Vector=$Engine.Owner.Velocity), B=100)
     True Float=0
     False Float=1
    )
   )
   &EmitterState(Life Cycle Mode=::NewEnumerator1, Loop Behavior=::NewEnumerator1, Loop Duration=1.5, UseLoopDelay=false)
   &SpawnRate(Spawn Probability=$StackContext.InterpolateOverTime.InterpolateOverTime_Float.Moving Average, SpawnRate=10, Use Spawn Probability=true)
  ]
  particleSpawnScript [
   &PartitionParticles()
   &InitializeParticle(
    Color=<rgba 3 0.7919 0.3503 1>
    Lifetime Max=1
    Lifetime Mode=::NewEnumerator1
    Sprite Size Max=<0.5 3>
    Sprite Size Min=<0.3 4>
    Sprite Size Mode=::NewEnumerator2
    UsePositionOffset=false
    Write Material Random=true
   )
   &SkeletalMeshLocation(
    Bone / Socket Index=$Particles.PartitionParticles.Partition
    Bone Sampling Mode=::NewEnumerator4
    MaskSpawnGroup=false
    Sampled Position Offset=&AddVectorToPosition(
     Position=&SelectVectorFromArray(
      Direct Array Index=$Particles.PartitionParticles.Partition
      Vector Selection Array=DataInterface() {propertyValues {FloatData [<-7 5 0>, <7 -5 0>]}}
     )
     Vector=<0 0 0>
    )
    Skeletal Mesh=$System.Skeletal Mesh
   )
   &ShapeLocation(
    Handle Radius=0
    Large Radius=8
    MaskSpawnGroup=false
    Non Uniform Scale=<0.6 1.1 1>
    Override Seed=false
    Rotation Axis=<0 0 1>
    Rotation Mode=::NewEnumerator1
    Shape Primitive=::NewEnumerator3
    UseOwnerScale=true
   )
   &CameraOffset(Camera Offset Amount=-10)
  ]
  particleUpdateScript [
   &ParticleState()
   /SetVariables(Particles.Glow03_Color=&MakeLinearColorFromVectorAndFloat(
     "Vector (RGB)"=&MakeVectorFromLinearColorRGB(Color=&MultiplyLinearColor(Linear Color=<rgba 1 0.0833 0 0.2>, Scale Factor=&MakeVector4_FromFloat(Float=1)))
     "Float (Alpha)"=&Multiply_Float(
      A=&Multiply_Float(
       A=0.2
       B=&MakeCustomFloatFromBool(
        Bool=&SetBoolByFloatComparison(A=&VectorLength(Vector=$Engine.Owner.Velocity), B=100)
        True Float=0
        False Float=1
       )
      )
      B=&MakeCustomFloatFromBool(Bool=$User.bisFootHidden, True Float=0, False Float=1)
     )
    )) {isSetParameters=true, module=SetVariables_#2}
   /SetVariables(Particles.Glow03_Size=&Multiply_Vector2D_ByFloat(
     Float=&FloatFromCurve(
      FloatCurve=DataInterface() {propertyValues {Curve {keys [
          <at 0 0 RCIM_Cubic>
          <at 0.2 1.3 RCIM_Cubic>
          {arriveTangent=-0.1423, interpMode=RCIM_Cubic, leaveTangent=-0.1423, tangentMode=RCTM_User, time=1, value=0}
         ]}}}
      Scale Curve=13
     )
     Vector2D=<0.15 0.4>
    )) {isSetParameters=true, module=SetVariables_#3}
   &ScaleColor(
    Scale Alpha=&FloatFromCurve(
     FloatCurve=DataInterface() {propertyValues {Curve {keys [<at 0.3 1>,<at 0.4 1>,<at 1 1>]}}}
     Scale Curve=&Multiply_Float(
      A=&MakeCustomFloatFromBool(
       Bool=&SetBoolByFloatComparison(A=&VectorLength(Vector=$Engine.Owner.Velocity), B=100)
       True Float=0
       False Float=1
      )
      B=&MakeCustomFloatFromBool(Bool=$User.bisFootHidden, True Float=0, False Float=1)
     )
    )
    ScaleA=true
    ScaleRGB=true
   )
   &ScaleSpriteSize(Scale Factor=&Vector2DFromFloat(Value=&FloatFromCurve(
      FloatCurve=DataInterface() {propertyValues {Curve {keys [
          <at 0 0 RCIM_Cubic>
          {interpMode=RCIM_Cubic, leaveTangent=-3.0296, tangentMode=RCTM_Break, time=0.2, value=1.2}
          {arriveTangent=-2.8894, interpMode=RCIM_Cubic, leaveTangent=-2.8894, tangentMode=RCTM_User, time=1, value=0.1}
         ]}}}
      Scale Curve=0.4
     ))) {module=ScaleSpriteSize001}
   &CurlNoiseForce(
    Noise Frequency=300
    Noise Strength=&Multiply_Float(A=100, B=1)
    Pan Noise=true
    Pan Noise Field=<0 0 1>
    Random Seed=10
   )
   &Drag(Drag=3, Use Linear Drag=true, Use Rotational Drag=false)
   &GravityForce(Gravity=<0 0 100>)
   &InheritSourceMovement(
    Applied Position Scale=&Multiply_VectorByFloat(Vector=<1 1 1>, Float=0.8)
    ApplyForce=false
    ApplyPosition=true
    ApplyPositionFalloff=false
   )
   &SolveForcesAndVelocity(Clamp Velocity=false, Limit Acceleration=false)
  ]
  rendererClasses [&NiagaraSpriteRendererProperties]
  renderers [
   {rendererClass=&NiagaraSpriteRendererProperties}
   {rendererClass=&NiagaraSpriteRendererProperties, rendererIndex=1}
  ]
 }
 ParticleDecal001 {
  emitterUpdateScript [
   &EmitterState(Life Cycle Mode=::NewEnumerator1, Loop Behavior=::NewEnumerator1, Loop Duration=1, UseLoopDelay=false)
   &SpawnBurst_Instantaneous(Spawn Count=4, Use Spawn Probability=false, UseLoopCountLimit=false)
  ]
  particleSpawnScript [
   &PartitionParticles(Number Of Partitions=4)
   &InitializeParticle(Color=<rgba 5 0.6667 0 1>, Sprite Size Mode=::NewEnumerator3, Uniform Sprite Size=30, UsePositionOffset=false, Write Material Random=true)
   &CameraOffset(Camera Offset Amount=25)
  ]
  particleUpdateScript [
   &ParticleState()
   &SkeletalMeshLocation(
    Bone / Socket Index=&SelectIntFromArray(Direct Array Index=$Particles.PartitionParticles.Partition, Int Selection Array=DataInterface() {propertyValues {IntData [2,3,4,5]}})
    Bone Sampling Mode=::NewEnumerator4
    MaskSpawnGroup=false
    Sampled Position Offset=&SelectVectorFromArray(
     Direct Array Index=$Particles.PartitionParticles.Partition
     Vector Selection Array=DataInterface() {propertyValues {FloatData [<10 0 0>, <-10 0 0>, <10 0 0>, <-10 0 0>]}}
    )
    Skeletal Mesh=$System.Skeletal Mesh
   )
   &ScaleColor(
    Scale Alpha=&FloatFromCurve(FloatCurve=DataInterface() {propertyValues {Curve {keys [
         <at 0 0 RCIM_Cubic>
         {arriveTangent=-0.8734, interpMode=RCIM_Cubic, leaveTangent=-0.8734, tangentMode=RCTM_User, time=0.1954, value=0.9441}
         {arriveTangent=-0.7911, tangentMode=RCTM_User, time=1, value=0}
        ]}}})
    ScaleA=true
    ScaleRGB=true
   )
  ]
  rendererClasses [&NiagaraSpriteRendererProperties]
  renderers [{rendererClass=&NiagaraSpriteRendererProperties}]
 }
}
summary {
 systemName=NS_EssayViewPier_Skin
 emitters [
  {bEnabled=false, emitterName=CalfSparks, rendererClasses [&NiagaraSpriteRendererProperties]}
  {emitterName=ArmSparksMovement, rendererClasses [&NiagaraSpriteRendererProperties]}
  {bEnabled=false, emitterName=CalfFire, rendererClasses [&NiagaraMeshRendererProperties]}
  {emitterName=BurstSparksFollow, simTarget=GPUComputeSim, rendererClasses [&NiagaraSpriteRendererProperties]}
  {emitterName=BurstArmSparks, simTarget=GPUComputeSim, rendererClasses [&NiagaraSpriteRendererProperties]}
  {emitterName=ParticleDecal001, rendererClasses [&NiagaraSpriteRendererProperties]}
  {emitterName=BurstSparksFollow001, simTarget=GPUComputeSim, rendererClasses [&NiagaraSpriteRendererProperties]}
  {emitterName=BurstArmSparks001, simTarget=GPUComputeSim, rendererClasses [&NiagaraSpriteRendererProperties]}
 ]
 userVariables [
  User.bisFootHidden {type=NiagaraBool}
  User.bIsFrontEndPreview {type=NiagaraBool, defaultValue=true}
 ]
}
