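# Material function that adds a timed, fiery reactive glow to the EssayViewPier character's incoming material attributes.
# Cylindrically projected, panned fire, smoke and dust noise uses explicit derivatives, alongside Fresnel rims, normal-derived lines and gradient-swept tattoos.
# A fade-in/out envelope and a frontend alpha override provide reactivity controls.
# Static switches select head, eye, face-accessory/hair or body output, changing emissive, plus opacity (disabling subsurface profiling) and body base color.
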MaterialFunction /BRCosmetics/Characters/Player/Male/Medium/Bodies/M_MED_EssayViewPier/Materials/MF_EssayViewPier_Reactive
graph {nodes [
  n1 = NamedRerouteUsage()
  n2 = NamedRerouteUsage()
  n3 = NamedRerouteUsage()
  n4 = VectorParameter() {group=ColorShading, parameterName=Whiteye, defaultValue=<rgba 1 1 1 10>}
  n5 = TextureSampleParameter2D() {
   group=Reactivity
   parameterName=Head_EM_Mask
   samplerType=SAMPLERTYPE_Masks
   texture=/BRCosmetics/Characters/Player/Male/Medium/Bodies/M_MED_EssayViewPier/Textures/T_essayviewPier_Head_TechArt
  }
  n6 = OneMinus(None=@n5.G)
  n7 = NamedRerouteUsage()
  n8 = StaticSwitchParameter(True=Add(A=@n3, B=Multiply(A=@n4.RGB, B=@n4.RGB)), False=Add(A=Multiply(A=@n3, B=@n6), B=Multiply(A=@n5.R, B=@n7))) {group=Reactivity, parameterName=IsEye}
  n9 = NamedRerouteUsage()
  n10 = LinearInterpolate(A=@n2, B=Add(A=@n2, B=Multiply(A=@n8, B=@n9)), Alpha=NamedRerouteUsage())
  n11 = LinearInterpolate(A=NamedRerouteUsage(), Alpha=NamedRerouteUsage()) {constB=0, desc=Disable Subsurface Profiling}
  n12 = NamedRerouteUsage()
  n13 = NamedRerouteUsage()
  n14 = SetMaterialAttributes!(MaterialAttributes=NamedRerouteUsage(), EmissiveColor=LinearInterpolate(A=@n12, B=Add(A=@n12, B=@n13), Alpha=NamedRerouteUsage()) {constB=0}) {attributeSetTypes [#1]}
  n15 = FunctionInput() {inputName=Material Input, inputType=FunctionInput_MaterialAttributes}
  n16 = GetMaterialAttributes(None=@n15) {attributeGetTypes [
    #1
    #2
    #3
    #4
   ]}
  n17 = NamedRerouteUsage()
  n18 = NamedRerouteUsage()
  n19 = SetMaterialAttributes!(
   MaterialAttributes=@n16.MaterialAttributes
   EmissiveColor=NamedRerouteUsage()
   BaseColor=LinearInterpolate(A=@n16.BaseColor, B=LinearInterpolate(A=@n16.BaseColor, Alpha=@n17) {constB=0}, Alpha=Power(Base=@n18) {constExponent=3})
   Opacity=LinearInterpolate(A=@n16.Opacity, Alpha=NamedRerouteUsage()) {constB=0, desc=Disable Subsurface Profiling}
  ) {attributeSetTypes [#1, #2, #3]}
  FunctionOutput(None=StaticSwitchParameter(
    True=SetMaterialAttributes!(MaterialAttributes=@n1, EmissiveColor=@n10, Opacity=@n11) {attributeSetTypes [#1, #3]}
    False=StaticSwitchParameter(True=@n14, False=@n19) {group=Reactivity, parameterName=IsFaceAcc}
   ) {group=Reactivity, parameterName=IsHead})
  NamedRerouteDeclaration(None=FunctionInput() {inputName=PreskinnedPOS, sortPriority=2}) {name=PreskinnedPOS, nodeColor=<rgba 0 1 0.3438 1>}
  n20 = ComponentMask(None=NamedRerouteUsage()) {g=true}
  n21 = TextureSampleParameter2D(UVs=AppendVector(A=NamedRerouteDeclaration(None=@n20) {name=VerticalGradient, nodeColor=<rgba 1 0.9609 0 1>}, B=Constant() {r=0.5})) {
   group=ColorShading
   parameterName=TattoGradient
   texture=/BRCosmetics/Characters/Player/Male/Medium/Bodies/M_MED_EssayViewPier/Textures/T_EssayView_TattooColor
  }
  NamedRerouteDeclaration(None=@n21.RGB) {name=GradientColor, nodeColor=<rgba 0.7812 1 0 1>}
  n22 = VectorParameter() {group=ColorShading, parameterName=FresnelColorDark, defaultValue=<rgba 1 0.0493 0 1>}
  n23 = Multiply(A=@n22.RGB, B=@n22.RGB)
  NamedRerouteDeclaration(None=@n23) {name=FresnelColor, nodeColor=<rgba 0.6641 1 0 1>}
  n24 = Power(Base=NamedRerouteUsage()) {constExponent=0.8}
  n25 = OneMinus(None=Frac(None=NamedRerouteUsage()))
  n26 = Constant() {r=1}
  n27 = Constant()
  n28 = Add(
   A=Multiply(A=NamedRerouteUsage(), B=ScalarParameter() {defaultValue=1, group=ColorShading, parameterName=GradientEmissiveIntensity})
   B=NamedRerouteUsage()
  )
  n29 = NamedRerouteUsage()
  n30 = Multiply(
   A=LinearInterpolate(A=SphereMask(A=@n24, B=@n25) {attenuationRadius=0.2, hardnessPercent=0}, Alpha=If(A=@n24, B=@n25, "A > B"=@n26, "A == B"=@n26, "A < B"=@n27))
   B=Add(A=NamedRerouteUsage(), B=Add(A=@n28, B=@n29))
  )
  n31 = TextureSampleParameter2D() {
   group=ColorShading
   parameterName=TattooMask
   samplerType=SAMPLERTYPE_Masks
   texture=/BRCosmetics/Characters/Player/Male/Medium/Bodies/M_MED_EssayViewPier/Textures/T_EssayViewPier_Body_TattooMask
  }
  NamedRerouteDeclaration(None=StaticSwitchParameter(True=Multiply(A=@n30, B=@n31.RGB), False=Constant()) {defaultValue=true, group=Reactivity, parameterName=UseTattooEM}) {name=TattooEM, nodeColor=<rgba 0 0.5078 1 1>}
  n32 = Saturate(None=LinearInterpolate(
    A=NamedRerouteUsage()
    B=ScalarParameter() {group=Reactivity, parameterName=ReactivityAlpha}
    Alpha=ScalarParameter() {group=Reactivity, parameterName=IsInFrontend}
   ))
  NamedRerouteDeclaration(None=Step(X=@n32) {constY=0.01}) {name=EnableReactivity, nodeColor=<rgba 0 0.6719 1 1>}
  n33 = NamedRerouteUsage()
  n34 = ComponentMask(None=NamedRerouteUsage()) {b=true}
  NamedRerouteDeclaration(None=Saturate(None=Subtract(A=@n33, B=@n34))) {name=EmissiveLines, nodeColor=<rgba 1 0 0.3047 1>}
  n35 = NamedRerouteUsage()
  n36 = ScalarParameter() {defaultValue=2, desc=UV must be integer to avoid seam, group=ColorShading, parameterName=UV_Tiling01}
  n37 = ScalarParameter() {defaultValue=0.1, group=ColorShading, parameterName=FireSpeedX01}
  n38 = ScalarParameter() {defaultValue=0.3, group=ColorShading, parameterName=FireSpeedY01}
  n39 = TextureSample!(
   UVs=Panner(Coordinate=Multiply(A=@n35, B=@n36), Speed=AppendVector(A=@n37, B=@n38)) {speedX=0.2, speedY=-0.2}
   Tex=NamedRerouteUsage()
   "DDX(UVs)"=NamedRerouteUsage()
   "DDY(UVs)"=NamedRerouteUsage()
   Apply View MipBias
  ) {
   mipValueMode=TMVM_Derivative
   samplerSource=SSM_Wrap_WorldGroupSettings
   texture=/BRCosmetics/Characters/Player/Male/Medium/Bodies/M_MED_EssayViewPier/Textures/T_FireNoisesPacked
  }
  NamedRerouteDeclaration(None=Multiply(A=NamedRerouteUsage(), B=Add(A=@n39.G, B=@n39.G))) {name=ComplexTattooColor, nodeColor=<rgba 1 0.8906 0 1>}
  n40 = MaterialFunctionCall!(
   Tile=VectorParameter() {group=UV, parameterName=UV_OGTiling, defaultValue=<rgba 3 3 0 1>}.RGB
   DistortionAdjustment=ScalarParameter() {defaultValue=-5, group=UV, parameterName=DistortIntensity}
  ) {materialFunction=&/MF_CylindricalProjUV_Complex}
  NamedRerouteDeclaration(None=@n40.UVs) {name=CylinderUV, nodeColor=<rgba 1 0 0.2578 1>}
  n41 = NamedRerouteUsage()
  NamedRerouteDeclaration(None=Multiply(A=Multiply(A=@n39.B, B=@n41), B=ScalarParameter() {defaultValue=10, group=ColorShading, parameterName=FireEmissiveIntensity})) {name=FirstLayerFire, nodeColor=<rgba 0.8281 1 0 1>}
  n42 = NamedRerouteUsage()
  n43 = ScalarParameter() {defaultValue=2, desc=UV must be integer to avoid seam, group=ColorShading, parameterName=UV_Tiling02}
  n44 = ScalarParameter() {defaultValue=0.2, group=ColorShading, parameterName=FireSpeedX02}
  n45 = ScalarParameter() {defaultValue=0.4, group=ColorShading, parameterName=FireSpeedY02}
  n46 = TextureSample(
   UVs=Panner(Coordinate=Multiply(A=@n42, B=@n43), Speed=AppendVector(A=@n44, B=@n45)) {speedX=0.2, speedY=-0.2}
   Tex=NamedRerouteUsage()
   "DDX(UVs)"=NamedRerouteUsage()
   "DDY(UVs)"=NamedRerouteUsage()
  ) {
   mipValueMode=TMVM_Derivative
   samplerSource=SSM_Wrap_WorldGroupSettings
   texture=&T_FireNoisesPacked
  }
  n47 = NamedRerouteUsage()
  NamedRerouteDeclaration(None=Multiply(A=ScalarParameter() {defaultValue=3, group=ColorShading, parameterName=SmokeEmissiveIntensity}, B=Multiply(A=@n46.G, B=@n47))) {name=SecondLayerFire, nodeColor=<rgba 0.5938 1 0 1>}
  n48 = TextureObjectParameter() {
   group=ColorShading
   parameterName=CustomMask
   texture=/BRCosmetics/Characters/Player/Male/Medium/Bodies/M_MED_EssayViewPier/Textures/T_FireNoises_Custom
  }
  NamedRerouteDeclaration(None=@n48) {name=CustomMask, nodeColor=<rgba 1 0.8906 0 1>}
  n49 = NamedRerouteUsage()
  n50 = ScalarParameter() {defaultValue=1, desc=UV must be integer to avoid seam, group=ColorShading, parameterName=UV_Tiling03}
  n51 = ScalarParameter() {defaultValue=0.2, group=ColorShading, parameterName=FireSpeedX03}
  n52 = ScalarParameter() {defaultValue=0.4, group=ColorShading, parameterName=FireSpeedY03}
  n53 = TextureSample(
   UVs=Panner(Coordinate=Multiply(A=@n49, B=@n50), Speed=AppendVector(A=@n51, B=@n52)) {speedX=0.2, speedY=-0.2}
   Tex=NamedRerouteUsage()
   "DDX(UVs)"=NamedRerouteUsage()
   "DDY(UVs)"=NamedRerouteUsage()
  ) {
   mipValueMode=TMVM_Derivative
   samplerSource=SSM_Wrap_WorldGroupSettings
   texture=&T_FireNoises_Custom
  }
  n54 = Multiply(A=Multiply(A=Power(Base=@n46.R) {constExponent=1.5}, B=@n53.R), B=NamedRerouteUsage())
  n55 = Power(Base=Frac(None=NamedRerouteUsage())) {constExponent=0.5}
  NamedRerouteDeclaration(None=Multiply(A=ScalarParameter() {defaultValue=10, group=ColorShading, parameterName=DustEmissiveIntensity}, B=Multiply(A=@n54, B=@n55))) {name=ThirdlayerEmissive, nodeColor=<rgba 0.5703 0 1 1>}
  n56 = TextureObjectParameter() {
   group=ColorShading
   parameterName=FireTexture
   texture=&T_FireNoisesPacked
  }
  NamedRerouteDeclaration(None=@n56) {name=FireTexture, nodeColor=<rgba 1 0.2578 0 1>}
  NamedRerouteDeclaration(None=@n32) {name=ReactiveAlpha, nodeColor=<rgba 0.0078 0 1 1>}
  n57 = NamedRerouteUsage()
  n58 = Fresnel(ExponentIn=ScalarParameter() {defaultValue=1, group=ColorShading, parameterName=BrightFrenselPOW}, Normal=NamedRerouteUsage()) {baseReflectFraction=0.1, exponent=1}
  n59 = Distance(
   A=MaterialFunctionCall(Normal=NamedRerouteUsage(), Flatness=Constant() {r=-3}) {materialFunction=&FlattenNormal}.Result
   B=Constant3Vector() {constant=<rgba 0 0 1 0>}
  )
  n60 = Multiply(A=ScalarParameter() {defaultValue=4, group=ColorShading, parameterName=EmissiveLinesIntensity}, B=@n59)
  n61 = NamedRerouteUsage()
  n62 = Add(A=Multiply(A=Multiply(A=@n57, B=@n58), B=NamedRerouteUsage()), B=Multiply(A=NamedRerouteUsage(), B=Multiply(A=@n60, B=@n61)))
  n63 = NamedRerouteUsage()
  NamedRerouteDeclaration(None=Add(A=NamedRerouteUsage(), B=Add(A=@n62, B=@n63))) {name=PostEM, nodeColor=<rgba 0.3125 1 0 1>}
  NamedRerouteDeclaration(None=@n16.EmissiveColor) {name=Emissive_IN, nodeColor=<rgba 0 1 0.3203 1>}
  n64 = NamedRerouteUsage()
  n65 = OneMinus(None=NamedRerouteUsage())
  n66 = NamedRerouteUsage()
  n67 = NamedRerouteUsage()
  NamedRerouteDeclaration(None=Add(A=Multiply(A=@n64, B=@n65), B=Multiply(A=@n66, B=@n67))) {name=EmissiveOUT, nodeColor=<rgba 1 0 0 1>}
  NamedRerouteDeclaration(None=@n15) {name=Material_IN, nodeColor=<rgba 0 0.3437 1 1>}
  n68 = TextureSampleParameter2D() {
   group=Reactivity_Hair
   parameterName=ReactiveMask_FaceAcc
   samplerType=SAMPLERTYPE_Masks
   texture=/BRCosmetics/Characters/Player/Male/Medium/Bodies/M_MED_EssayViewPier/Textures/T_essayviewPier_FaceAcc_TechArt
  }
  n69 = SmoothStep(
   Min=ScalarParameter() {group=Reactivity_Hair, parameterName=HairMin}
   Max=ScalarParameter() {defaultValue=1, group=Reactivity_Hair, parameterName=HairMax}
   Value=Power(Base=@n68.R) {constExponent=3}
  )
  n70 = NamedRerouteUsage()
  n71 = Fresnel(Normal=NamedRerouteUsage()) {baseReflectFraction=0.5, exponent=2}
  n72 = VectorParameter() {group=Reactivity_Hair, parameterName=HairEM_Dark, defaultValue=<rgba 1 0.0597 0 1>}
  n73 = VectorParameter() {group=Reactivity_Hair, parameterName=HairEM_Bright, defaultValue=<rgba 1 0.358 0 1>}
  n74 = LinearInterpolate(A=Multiply(A=@n72.RGB, B=@n72.RGB), B=Multiply(A=@n73.RGB, B=@n73.RGB), Alpha=NamedRerouteUsage())
  n75 = Multiply(A=NamedRerouteUsage(), B=ScalarParameter() {defaultValue=10, group=Reactivity_Hair, parameterName=HairFireIntensity})
  n76 = NamedRerouteDeclaration(None=Add(A=Multiply(A=@n71, B=@n74), B=Multiply(A=@n74, B=@n75))) {name=HairGradientColor, nodeColor=<rgba 1 0 0.75 1>}
  n77 = Multiply(A=@n68.G, B=Multiply(A=@n69, B=Add(A=@n70, B=@n76)))
  n78 = NamedRerouteUsage()
  n79 = NamedRerouteUsage()
  n80 = Add(A=@n77, B=Multiply(A=NamedRerouteUsage(), B=Multiply(A=@n78, B=@n79)))
  NamedRerouteDeclaration(None=Multiply(A=StaticSwitchParameter(True=@n80, False=@n77) {group=Reactivity, parameterName=IsWithHat}, B=NamedRerouteUsage())) {name=Hair_EM, nodeColor=<rgba 0.6172 1 0 1>}
  NamedRerouteDeclaration(None=@n16.Normal) {name=NormalIN, nodeColor=<rgba 1 0 0.2344 1>}
  n81 = NamedRerouteUsage()
  n82 = ScalarParameter() {group=ColorShading, parameterName=FlattenNormal}
  NamedRerouteDeclaration(None=Transform(None=MaterialFunctionCall(Normal=@n81, Flatness=@n82) {materialFunction=&FlattenNormal}.Result)) {name=Normal_World, nodeColor=<rgba 1 0.7969 0 1>}
  NamedRerouteDeclaration(None=@n69) {name=GradientHair, nodeColor=<rgba 0.4297 1 0 1>}
  NamedRerouteDeclaration(None=@n40.DDX) {name=ddx, nodeColor=<rgba 1 0 0.4922 1>}
  NamedRerouteDeclaration(None=@n39.B) {name=FirePattern, nodeColor=<rgba 0.1016 0 1 1>}
  NamedRerouteDeclaration(None=Power(Base=@n68.B) {constExponent=2.2}) {name=HatMask_Constant, nodeColor=<rgba 0 0.2734 1 1>}
  NamedRerouteDeclaration(None=@n75) {name=FireColor, nodeColor=<rgba 1 0 0.8437 1>}
  NamedRerouteDeclaration(None=@n16.Opacity) {name=OpacityIN, nodeColor=<rgba 0.4766 1 0 1>}
  n83 = Time()
  n84 = ScalarParameter() {group=Reactivity, parameterName=ReactivityStartTime}
  n85 = ScalarParameter() {defaultValue=4.5, group=Reactivity, parameterName=ReactivityPeriod}
  n86 = Saturate(None=Divide(A=Subtract(A=@n83, B=@n84), B=@n85))
  n87 = Divide(A=Constant() {r=0.5}, B=@n85)
  n88 = Saturate(None=Multiply(A=@n86, B=Divide(B=@n87)))
  n89 = Constant()
  n90 = If(A=@n86, B=@n87, "A > B"=Constant() {r=1}, "A == B"=@n89, "A < B"=@n89)
  n91 = Divide(A=Constant() {r=1.5}, B=@n85)
  n92 = OneMinus(None=@n91)
  n93 = Saturate(None=Multiply(A=Subtract(A=@n86, B=@n92), B=Divide(B=@n91)) {constB=5})
  n94 = Constant() {r=1}
  n95 = Constant()
  NamedRerouteDeclaration(None=LinearInterpolate(A=LinearInterpolate(A=@n88, Alpha=@n90), B=OneMinus(None=@n93), Alpha=If(A=@n86, B=@n92, "A > B"=@n94, "A == B"=@n95, "A < B"=@n95))) {name=ComplexReactivity, nodeColor=<rgba 0 1 0.5781 1>}
  NamedRerouteDeclaration(None=@n40.DDY) {name=ddy, nodeColor=<rgba 0 0.8125 1 1>}
  NamedRerouteDeclaration(None=FunctionInput() {inputName=M, sortPriority=1}) {name=M_Mask, nodeColor=<rgba 0.0781 1 0 1>}
  n96 = VectorParameter() {group=ColorShading, parameterName=FresnelColorLight, defaultValue=<rgba 1 0.3 0 1>}
  n97 = LinearInterpolate(
   A=@n23
   B=Multiply(A=@n96.RGB, B=@n96.RGB)
   Alpha=Fresnel(
    ExponentIn=ScalarParameter() {defaultValue=1, group=ColorShading, parameterName=FresnelExp_Color}
    BaseReflectFractionIn=ScalarParameter() {group=ColorShading, parameterName=FresnelIn_Color}
   ) {baseReflectFraction=0, exponent=1}
  )
  n98 = ScalarParameter() {defaultValue=1, group=ColorShading, parameterName=FresnelExp}
  n99 = ScalarParameter() {group=ColorShading, parameterName=FresnelIn}
  n100 = NamedRerouteUsage()
  n101 = Multiply(
   A=SmoothStep(
    Min=ScalarParameter() {defaultValue=0.4, group=ColorShading, parameterName=Fresnel_Min}
    Max=ScalarParameter() {defaultValue=0.5, group=ColorShading, parameterName=Fresnel_Max}
    Value=Fresnel(ExponentIn=@n98, BaseReflectFractionIn=@n99, Normal=@n100) {baseReflectFraction=0, exponent=1}
   )
   B=ScalarParameter() {defaultValue=1, group=ColorShading, parameterName=GeneralFresnelIntensity}
  )
  n102 = ScalarParameter() {defaultValue=3, group=ColorShading, parameterName=FlameIntensity}
  n103 = ScalarParameter() {defaultValue=1, group=ColorShading, parameterName=FresnelExp02}
  n104 = ScalarParameter() {group=ColorShading, parameterName=FresnelIn02}
  n105 = NamedRerouteUsage()
  n106 = Multiply(
   A=SmoothStep(
    Min=ScalarParameter() {defaultValue=0.2, group=ColorShading, parameterName=Flame_Min}
    Max=ScalarParameter() {defaultValue=0.5, group=ColorShading, parameterName=Flame_Max}
    Value=NamedRerouteUsage()
   )
   B=SmoothStep(
    Min=ScalarParameter() {defaultValue=0.4, group=ColorShading, parameterName=Fresnel_Min02}
    Max=ScalarParameter() {defaultValue=0.5, group=ColorShading, parameterName=Fresnel_Max02}
    Value=Fresnel(ExponentIn=@n103, BaseReflectFractionIn=@n104, Normal=@n105) {baseReflectFraction=0, exponent=1}
   )
  )
  n107 = StaticSwitchParameter(True=@n101, False=Add(A=Multiply(A=@n102, B=@n106), B=@n101)) {group=Reactivity, parameterName=IsHead}
  n108 = Constant() {r=1}
  n109 = ComponentMask(None=NamedRerouteUsage()) {b=true}
  NamedRerouteDeclaration(None=Multiply(A=Multiply(A=@n97, B=@n107), B=StaticSwitchParameter(True=@n108, False=@n109) {group=Reactivity, parameterName=IsFaceAcc})) {name=Fresnel_EM, nodeColor=<rgba 1 0 0.9609 1>}
  NamedRerouteDeclaration(None=@n53.G) {name=Flame, nodeColor=<rgba 0.1719 1 0 1>}
  n110 = TextureSample!(UVs, Tex=NamedRerouteUsage(), Apply View MipBias) {
   samplerSource=SSM_Wrap_WorldGroupSettings
   texture=&T_FireNoises_Custom
  }
  NamedRerouteDeclaration(None=@n110.B) {name=EmissiveMask, nodeColor=<rgba 0 1 0.8594 1>}
  NamedRerouteUsage()
  Power(Base=@n59) {constExponent=2.2}
 ]}
