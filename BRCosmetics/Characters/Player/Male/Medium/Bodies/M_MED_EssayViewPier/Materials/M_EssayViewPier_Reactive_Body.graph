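# Body material for the EssayViewPier character skin, adding that skin's reactive effect to the 2023 character parent shading.
# It chains shared PreFX, the skin's Reactive function and shared PostFX, passing material attributes along.
# A UseFacialExpressionCinematic static switch optionally inserts the advanced facial expression function before the reactive stage.
# Shading model comes from the material expression, and usage covers skeletal meshes, morph targets and mesh deformers.
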Material /BRCosmetics/Characters/Player/Male/Medium/Bodies/M_MED_EssayViewPier/Materials/M_EssayViewPier_Reactive_Body
graph {
 properties {
  bUseMaterialAttributes=true
  bUsedWithMeshDeformer=true
  bUsedWithMorphTargets=true
  bUsedWithSkeletalMesh=true
  pixelDepthOffsetMode=PDOM_Legacy
  shadingModel=MSM_FromMaterialExpression
  usedShadingModels=Default Lit | Subsurface | Clear Coat | Subsurface Profile | Cloth
 }
 nodes [
  n1 = MaterialFunctionCall() {materialFunction=&MF_CharacterParent_2023_PreFX}
  n2 = MaterialFunctionCall!(Material=@n1.Result) {materialFunction=/Game/Global/Materials/Character_Parent_2023/FacialExpression_Parent_2023/MF_FacialExpressionAdvanced_2023}
  MaterialFunctionCall(
   Material=MaterialFunctionCall!(
    Material Input=StaticSwitchParameter(True=@n2.Result, False=@n1.Result) {group=Control, parameterName=UseFacialExpressionCinematic}
    M=@n1.M
    PreskinnedPOS=@n1.M
   ) {materialFunction=&/MF_EssayViewPier_Reactive}.Result
   M=@n1.M
   Pre Skinned Local Position=@n1.M
   Distance Blend=@n1.M
   Fresnel=@n1.M
  ) {materialFunction=&MF_CharacterParent_2023_PostFX}
 ]
}
