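# Parent eye material for the EssayViewPier male medium character's reactive variant, built on material attributes for skeletal meshes.
# The shared 2023 eye function feeds the tech art overrides function, whose result enters the EssayViewPier reactive function.
# That function also receives a constant white M input and the pre-skinned position passed through a vertex interpolator, so it has access to rest-pose coordinates.
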Material /BRCosmetics/Characters/Player/Male/Medium/Bodies/M_MED_EssayViewPier/Materials/M_Eyes_Parent_EssayViewPier_Reactive
graph {
 properties {
  bUseMaterialAttributes=true
  bUsedWithMeshDeformer=true
  bUsedWithSkeletalMesh=true
  pixelDepthOffsetMode=PDOM_Legacy
  shadingModel=MSM_FromMaterialExpression
  usedShadingModels=Default Lit
 }
 nodes [MaterialFunctionCall!(
   Material Input=MaterialFunctionCall(In=MaterialFunctionCall() {materialFunction=/Game/Global/Materials/Character_Parent_2023/Eye_Parent_2023/MF_eyes_2023}.Result) {materialFunction=&MF_TechArtOverrides}.Result
   M=Constant3Vector() {constant=<rgba 1 1 1 0>}
   PreskinnedPOS=VertexInterpolator(VS=PreSkinnedPosition()).PS
  ) {materialFunction=&/MF_EssayViewPier_Reactive}]
}
