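# Control Rig for the M_MED_EssayViewPier male player body that runs a shared deform rig, then corrects the watch, elbow pad, belt bullets and chain bones.
# Watch and bullet offsets are weighted by deformRig animation attributes, with the watch also reading right-hand rotation.
# Parent constraints blend elbow pad and bullets between body bones.
# The chain offset multiplies head and right-upper-arm spherical pose readers, firing only when both poses match.
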ControlRigBlueprint /BRCosmetics/Characters/Player/Male/Medium/Bodies/M_MED_EssayViewPier/Meshes/M_MED_EssayViewPier_CtrlRig
graphs [
 New Function_ContainedGraph {nodes [
   RigVMFunctionEntryNode()
   > &RigUnit_ModifyTransforms(
    ItemToModify=((Item=(Type=Bone,Name="watch_01"),Transform=(Rotation=(X=0,Y=-0.0964,Z=0,W=0.9953),Translation=(X=0.9754,Y=0,Z=0),Scale3D=(X=1,Y=1,Z=1.07))))
    WeightMinimum=0
    Weight=RigVMDispatchNode(BoneName=root, Default=1, Name=deformRig_r_wrist_up_alpha) {template="DISPATCH_RigDispatch_GetAnimAttribute(in Name,in BoneName,in Default,out Value,out Found)"}.Value
   )
   Remap = &RigVMFunction_MathDoubleRemap(
    SourceMaximum=1
    SourceMinimum=0.4
    TargetMaximum=1
    TargetMinimum=0
    bClamp=False
    Value=RigVMDispatchNode(BoneName=root, Default=1, Name=deformRig_r_wrist_dn_alpha) {template="DISPATCH_RigDispatch_GetAnimAttribute(in Name,in BoneName,in Default,out Value,out Found)"}.Value
   )
   > &RigUnit_ModifyTransforms(
    ItemToModify=((Item=(Type=Bone,Name="watch_01"),Transform=(Rotation=(X=0,Y=0,Z=0,W=1),Translation=(X=2.0287,Y=0,Z=0.0044),Scale3D=(X=1,Y=1,Z=1))))
    WeightMinimum=0
    Weight=@Remap.Result
   )
   RigVMFunction_MathQuaternionToEuler = &RigVMFunction_MathQuaternionToEuler(
    RotationOrder=ZYX
    Value=&RigVMFunction_MathTransformMakeRelative(
     Global=&RigUnit_GetTransform(Item=(Type=Bone,Name="hand_r"), Space=LocalSpace).Transform
     Parent=&RigUnit_GetTransform(Item=(Type=Bone,Name="hand_r"), Space=LocalSpace, bInitial=true).Transform
    ).Local.Rotation
   )
   > &RigUnit_ModifyTransforms(
    ItemToModify=((Item=(Type=Bone,Name="watch_00"),Transform=(Rotation=(X=0,Y=0,Z=0,W=-1),Translation=(X=1.2688,Y=0,Z=0),Scale3D=(X=1,Y=1,Z=1))))
    WeightMinimum=0
    Weight=&RigVMFunction_MathDoubleRemap(SourceMaximum=-60, SourceMinimum=-35, TargetMaximum=1, TargetMinimum=0, Value=@RigVMFunction_MathQuaternionToEuler.Result.Y).Result
   )
   > &RigUnit_ModifyTransforms(
    ItemToModify=((Item=(Type=Bone,Name="watch_00"),Transform=(Rotation=(X=0,Y=0,Z=0,W=1),Translation=(X=1.3324,Y=0,Z=0),Scale3D=(X=1,Y=1,Z=1))))
    WeightMinimum=0
    Weight=&RigVMFunction_MathDoubleRemap(SourceMaximum=60, SourceMinimum=35, TargetMaximum=1, TargetMinimum=0, Value=@RigVMFunction_MathQuaternionToEuler.Result.Y).Result
   )
   > RigVMFunctionReturnNode()
  ]}
 New Function_ContainedGraph {nodes [
   RigVMFunctionEntryNode()
   > &RigUnit_ParentConstraint(
    AdvancedSettings=(InterpolationType=Average,RotationOrderForFilter=XZY)
    Child=(Type=Bone,Name="elbowPad_00")
    Parents=((Item=(Type=Bone,Name="upperarm_r"),Weight=1),(Item=(Type=Bone,Name="lowerarm_r"),Weight=1))
    Weight=1
   )
   > RigVMFunctionReturnNode()
  ]}
 New Function_ContainedGraph {nodes [
   RigVMFunctionEntryNode()
   > &RigUnit_ParentConstraint(
    AdvancedSettings=(InterpolationType=Average,RotationOrderForFilter=XZY)
    Child=(Type=Bone,Name="base_beltBullets_00")
    Parents=((Item=(Type=Bone,Name="pelvis"),Weight=0.9863),(Item=(Type=Bone,Name="spine_02"),Weight=0.0056),(Item=(Type=Bone,Name="spine_01"),Weight=0.0049),(Item=(Type=Bone,Name="spine_03"),Weight=0.0033))
    Weight=1
   )
   > &RigUnit_ModifyTransforms(
    ItemToModify=((Item=(Type=Bone,Name="dyn_beltBullets_00"),Transform=(Rotation=(X=0,Y=0.1966,Z=0,W=-0.9805),Translation=(X=-1.0061,Y=0,Z=3.9811),Scale3D=(X=1,Y=1,Z=1))))
    WeightMinimum=0
    Weight=RigVMDispatchNode(BoneName=root, Default=1, Name=deformRig_r_thigh_up_alpha) {template="DISPATCH_RigDispatch_GetAnimAttribute(in Name,in BoneName,in Default,out Value,out Found)"}.Value
   )
   > RigVMFunctionReturnNode()
  ]}
 New Function_ContainedGraph {nodes [
   n1 = RigVMFunctionEntryNode()
   RigVMFunction_Sequence = &RigVMFunction_Sequence(ExecuteContext=@n1.ExecuteContext)
   n2 = &RigUnit_SphericalPoseReader(
    ActiveRegionScaleFactors=(PositiveWidth=1,NegativeWidth=0.5,PositiveHeight=0.75,NegativeHeight=0.75)
    ActiveRegionSize=0.1
    Debug=(bDrawDebug=false,bDraw2D=False,bDrawLocalAxes=False,DebugScale=25,DebugSegments=20,DebugThickness=0.25)
    DriverAxis=(X=0,Y=1,Z=0)
    DriverItem=(Type=Bone,Name="head")
    FalloffRegionScaleFactors=(PositiveWidth=1,NegativeWidth=0.5,PositiveHeight=1,NegativeHeight=1)
    FalloffSize=0.2
    FlipHeightScaling=False
    FlipWidthScaling=False
    OptionalParentItem=(Type=Bone,Name="None")
    RotationOffset=(X=90,Y=0,Z=-30)
    ExecutePin=@RigVMFunction_Sequence.A
   )
   > RigVMVariableNode(Value=@n2.OutputParam) {variable=head_SPR_look_up_alpha}
   > n3 = &RigUnit_SphericalPoseReader(
    ActiveRegionScaleFactors=(PositiveWidth=1,NegativeWidth=1,PositiveHeight=1,NegativeHeight=1)
    ActiveRegionSize=0.05
    Debug=(bDrawDebug=false,bDraw2D=False,bDrawLocalAxes=False,DebugScale=25,DebugSegments=20,DebugThickness=0.25)
    DriverAxis=(X=-1,Y=0,Z=0)
    DriverItem=(Type=Bone,Name="upperarm_r")
    FalloffRegionScaleFactors=(PositiveWidth=1,NegativeWidth=1,PositiveHeight=1,NegativeHeight=1)
    FalloffSize=0.2
    FlipHeightScaling=False
    FlipWidthScaling=False
    OptionalParentItem=(Type=Bone,Name="None")
    RotationOffset=(X=-115,Y=0,Z=25)
   )
   > RigVMVariableNode(Value=@n3.OutputParam) {variable=upperarm_r_SPR_aim_up_two_hands_alpha}
   Multiply = &RigVMFunction_MathDoubleMul(A=RigVMVariableNode() {variable=head_SPR_look_up_alpha}.Value, B=RigVMVariableNode() {variable=upperarm_r_SPR_aim_up_two_hands_alpha}.Value)
   > RigVMVariableNode(Value=@Multiply.Result) {variable=look_up_and_left_two_hands_multiplied_alpha}
   > &RigUnit_ModifyTransforms(
    ItemToModify=((Item=(Type=Bone,Name="base_chain_00"),Transform=(Rotation=(X=0.03,Y=0.0095,Z=-0.0003,W=-0.9995),Translation=(X=0,Y=0.3361,Z=-1.593),Scale3D=(X=1,Y=1,Z=1))))
    WeightMinimum=0
    Weight=RigVMVariableNode() {variable=look_up_and_left_two_hands_multiplied_alpha}.Value
   )
   > RigVMFunctionReturnNode()
  ]}
 RigVMFunctionLibrary {nodes [
   Chain = RigVMCollapseNode(ExecuteContext=())
   ElbowPad = RigVMCollapseNode(ExecuteContext=())
   HipBullets = RigVMCollapseNode(ExecuteContext=())
   Watch = RigVMCollapseNode(ExecuteContext=())
  ]}
 RigVMModel {nodes [
   RigUnit_BeginExecution = &RigUnit_BeginExecution()
   RigVMFunction_Sequence = &RigVMFunction_Sequence(ExecuteContext=@RigUnit_BeginExecution.ExecutePin)
   FN_DeformRig_FNC = RigVMFunctionReferenceNode(
    DeformRig_ Curve="(deformRig_l_upperarm_up_alpha_72_#1=False,deformRig_r_upperarm_up_alpha_73_#2=False,deformRig_l_upperarm_fr_alpha_74_#3=False,deformRig_r_upperarm_fr_alpha_75_CB51B8 ...(truncated)"
    DeformRig_Bool="(upperarm_1_#4=True,elbow_3_#5=True,wrist_7_#6=True,groin_8_#7=True,glute_17_#8=True,knee_9_40C408324CA261 ...(truncated)"
    DeformRig_Debug=(upperarm_debug_18_#4=False,wrist_debug_20_#6=False,lat_debug_26_#9=False,groinGlute_debug_27_#7=False)
    DeformRig_Value="(upperarm_up_18_#10=(Rotation=(X=0.0,Y=-0.0,Z=0.0,W=1.0),Translation=(X=0.0,Y=0.0,Z=8.0),Scale3D=(X=0.70,Y=1.0,Z=1.0)),upperarm_fwd_fr_19_#11=(Rotation=(X=0.0,Y=-0.0,Z=0.0,W=1.0),Translation=(X=0.0, ...(truncated)"
    ExecuteContext=@RigVMFunction_Sequence.A
   )
   Watch = RigVMFunctionReferenceNode(ExecuteContext=@RigVMFunction_Sequence.B)
   > ElbowPad = RigVMFunctionReferenceNode()
   > HipBullets = RigVMFunctionReferenceNode()
   > Chain = RigVMFunctionReferenceNode()
   setFKCtrl = RigVMFunctionReferenceNode(ExecuteContext=(), setFK_Ctrl=false)
  ]}
 RigVMModel Construction Event Graph {nodes [
   n1 = &RigUnit_PrepareForExecution()
   &RigUnit_HierarchyImportFromSkeleton(NameSpace=None, bIncludeCurves=False, bIncludeMeshSockets=False, bIncludeVirtualBones=True, ExecutePin=@n1.ExecutePin)
   addFKCtrl = RigVMFunctionReferenceNode(
    Custom_FKCtrl=()
    ExecuteContext=()
    Options_FKCtrl="(pelvis_1_#12=true,spine_3_#13=true,clavicle_5_#14=true,arm_7_#15=true,leg_9_#16=true,head_11_46DC15EF465CD2979 ...(truncated)"
    Shape_Settings="(pelvis_2_#17=(bVisible=True,Name=\"Hexagon_Thin\",Color=(R=0.0117,G=0.8281,B=1,A=1),Transform=(Rotation=(X=0,Y=-0.7071,Z=0,W=0.7071),Translation=(X=2,Y=0,Z=0),Scale3D=(X=5,Y=5,Z=7))),spine_4_#18=(bVi ...(truncated)"
   )
  ]}
 RigVMModel Rig {links [], nodes []}
]
variables [
 head_SPR_look_up_alpha {type=double, default=0}
 look_up_and_left_two_hands_multiplied_alpha {type=double, default=0}
 upperarm_r_SPR_aim_up_two_hands_alpha {type=double, default=0}
]
