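# Physics asset for the male medium player body M_MED_EssayViewPier, set up for a dangling chain, apparently a necklace.
# Two dyn_chain capsules hang in series from a small base_chain_00 body, with tight twist and uneven swing limits that favor one axis.
# Capsules on spine_04, spine_05 and a base_necklaceCollider bone give torso collision volumes.
# The chain base, necklace collider and spine links use fully free angular constraints, so only the chain joints are angularly restricted.
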PhysicsAsset /BRCosmetics/Characters/Player/Male/Medium/Bodies/M_MED_EssayViewPier/Meshes/M_MED_EssayViewPier_Physics
bodies {
 base_chain_00 [{shapeType=Capsule, length=0.505, radius=0.505, shapeName=base_chain_00_capsule, rotation=<rot 0 -90 0>}]
 base_necklaceCollider [{
   shapeType=Capsule
   length=18.5002
   radius=8.2523
   shapeName=base_necklaceCollider_capsule
   center=<0 -7.7835 -2.6448>
   rotation=<rot 0 -90 90>
  }]
 dyn_chain_00 [{
   shapeType=Capsule
   length=9.7613
   radius=0.6433
   shapeName=dyn_chain_00_capsule
   center=<6.4554 0.3519 -0.7985>
   rotation=<rot 82.1072 180 180>
  }]
 dyn_chain_01 [{shapeType=Capsule, length=7.5218, radius=1.1294, shapeName=dyn_chain_01_capsule, center=<4.6847 0 2.8993>, rotation=<rot 45 180 90>}]
 spine_04 [{shapeType=Capsule, length=20.8248, radius=7.1599, shapeName=spine_04_capsule, center=<10.1964 4.7143 0>}]
 spine_05 [{shapeType=Capsule, length=18.0233, radius=7.7002, shapeName=spine_05_capsule, center=<2.4265 1.6133 0>}]
}
constraints [
 {bone1Name=dyn_chain_00, bone2Name=base_chain_00, swing1LimitDegrees=5, swing2LimitDegrees=55, twistLimitDegrees=5}
 {bone1Name=dyn_chain_01, bone2Name=dyn_chain_00, swing1LimitDegrees=8, swing2LimitDegrees=35, twistLimitDegrees=1}
 {
  bone1Name=base_chain_00
  bone2Name=spine_05
  swing1LimitDegrees=45
  swing1Motion=Free
  swing2LimitDegrees=45
  swing2Motion=Free
  twistLimitDegrees=45
  twistMotion=Free
 }
 {
  bone1Name=spine_05
  bone2Name=spine_04
  swing1LimitDegrees=45
  swing1Motion=Free
  swing2LimitDegrees=45
  swing2Motion=Free
  twistLimitDegrees=45
  twistMotion=Free
 }
 {
  bone1Name=base_necklaceCollider
  bone2Name=spine_05
  swing1LimitDegrees=45
  swing1Motion=Free
  swing2LimitDegrees=45
  swing2Motion=Free
  twistLimitDegrees=45
  twistMotion=Free
 }
]
